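# Animation Blueprint for the InflateMate keytar instrument cosmetic, adding physics secondary motion to its skeleton.
# The AnimGraph runs the input pose through a RigidBody node simulated in handle bone space with an override physics asset.
# Update events set GravityOverride from skydiving and parachute state, though the node shows fixed gravity with no visible link to it.
# Customization load caches the pawn and backpack mesh; pickaxe and backpack-validity checks have no incoming execution.
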AnimBlueprint /SparksCosmetics/Instrument/Keytar/InflateMate/Meshes/InflateMate_Keytar_AnimBP
target_skeleton=&/InflateMate_Keytar_Skeleton
graphs [
 AnimGraph {nodes [
   n1 = RigidBody!(ComponentPose=Local To Component(LocalPose=Input Pose()), bAlphaBoolEnabled:bool, OverrideWorldGravity:Vector="0,0,-900", ExternalForce) -> (Pose) {
    alphaInputType=Bool
    bOverrideWorldGravity=true
    overridePhysicsAsset=&/InflateMate_Keytar_Physics
    simulationSpace=BaseBoneSpace
    alphaBoolBlend {blendOption=Linear}
    baseBoneRef {boneName=handle}
    componentAppliedLinearAccClamp=<3000 3000 3000>
    componentLinearAccScale=<0.4 0.4 0.4>
    componentLinearVelScale=<0.2 0.2 0.2>
    overrideWorldGravity {z=-900}
    simSpaceSettings {
     maxAngularAcceleration=500
     maxAngularVelocity=3
     maxLinearAcceleration=700
     maxLinearVelocity=500
     velocityScaleZ=0
     worldAlpha=0.5
     externalLinearDragV=<0.2 0.2 0.02>
    }
   }
   Output Pose(Result=Component To Local(ComponentPose=@n1))
  ]}
 EventGraph {nodes [
   n1 = Cast To PlayerPawn_Athena(Object=TryGetPawnOwner())
   > n2 = Set As Player Pawn Athena(As Player Pawn Athena:Player Pawn Athena Object Reference=@n1) -> (Output_Get:Player Pawn Athena Object Reference)
   n3 = Get CurrentWeaponList!(self:Fort Pawn Object Reference=@n2) -> (CurrentWeaponList:Array of Fort Weapon Object References)
   > n4 = Filter Array!(
    execute
    TargetArray:Array of Actor Object References=@n3
    FilterClass:Actor Class Reference=/Game/Weapons/FORT_Melee/Blueprints/B_Athena_Pickaxe_Generic.B_Athena_Pickaxe_Generic_C
   ) -> (then, FilteredArray:Array of B Athena Pickaxe Generic Object References)
   > "Set Instrument Pickaxe "("Instrument Pickaxe ":Array of B Athena Pickaxe Generic Object References=@n4) -> (Output_Get:Array of B Athena Pickaxe Generic Object References)
   > Set IsPickaxeEquipped(IsPickaxeEquipped:bool=Is Valid Index!(TargetArray:Array of Fort Weapon Object References=@n3, IndexToTest:int=1) -> (ReturnValue:bool)) -> (Output_Get:bool)
   Event BlueprintUpdateAnimation()
   > n5 = Branch(Condition=Get bIsSkyDiving())
   > n6 = Branch(Condition=Get bIsParachuteOpened())
   > Set GravityOverride(GravityOverride="0,0,-400")
   Set GravityOverride(execute=@n6.else, GravityOverride="0,0,500")
   Set GravityOverride(execute=@n5.else, GravityOverride="0,0,-900")
   Event OnCharacterCustomizationLoaded()
   > n7 = Cast To PlayerPawn_Athena(Object=TryGetPawnOwner())
   > Set As Player Pawn Athena(As Player Pawn Athena=@n7)
   > Set Backpack_SkelMesh(Backpack_SkelMesh:Skeletal Mesh Component Object Reference=GetSkeletalMeshForPartType(self=@n7, PartType=Backpack)) -> (Output_Get:Skeletal Mesh Component Object Reference)
   n8 = Is Valid(InputObject=Get Backpack_SkelMesh() -> (Backpack_SkelMesh:Skeletal Mesh Component Object Reference))
   Set IsBackpackEquipped(execute=@n8.Is Not Valid, IsBackpackEquipped:bool=false) -> (Output_Get:bool)
   Set IsBackpackEquipped(execute=@n8.Is Valid, IsBackpackEquipped=true)
   TryGetPawnOwner()
  ]}
]
variables [
 As Player Pawn Athena {type=PlayerPawn_Athena_C}
 Backpack_SkelMesh {type=SkeletalMeshComponent}
 GravityOverride {type=Vector}
 "Instrument Pickaxe " {type="array<B_Athena_Pickaxe_Generic_C>", default []}
 IsBackpackEquipped {type=boolean}
 IsPickaxeEquipped {type=boolean}
 NewVar {type=boolean}
]
