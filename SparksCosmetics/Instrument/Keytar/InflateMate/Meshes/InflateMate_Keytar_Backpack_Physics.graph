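# Physics asset for the InflateMate keytar backpack cosmetic, giving the balloon-style instrument capsule bodies for constrained, slightly flexible motion.
# The neck, top and bottom chains branch from the handle, with two headstock balloons on the neck and a tie on the second balloon.
# Angular limits are tight, with twist mostly locked.
# Fully angular-locked joints join top_03 to top_02 and bottom_02, closing the body into a loop.
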PhysicsAsset /SparksCosmetics/Instrument/Keytar/InflateMate/Meshes/InflateMate_Keytar_Backpack_Physics
bodies {
 bottom_01 [{shapeType=Capsule, length=7.4032, radius=2.5621, shapeName=bottom_01_capsule, center=<7 0.85 0>, rotation=<rot 0 89.9668 -89.3925>}]
 bottom_02 [{shapeType=Capsule, length=5.5, radius=9, shapeName=bottom_02_capsule, center=<9.8 1 0.5>}]
 handle [{shapeType=Capsule, length=2, radius=4}]
 headstock_balloon_01 [{shapeType=Capsule, length=4.8851, radius=4.0521, shapeName=headstock_balloon_01_capsule, center=<4 0 0>, rotation=<rot 7.6448 0 -90>}]
 headstock_balloon_02 [{shapeType=Capsule, length=3.5584, radius=2.7091, shapeName=headstock_balloon_02_capsule, center=<3.3 0 0>, rotation=<rot 0 89.8787 -88.4811>}]
 headstock_balloon_tie [{shapeType=Capsule, length=0.005, radius=1.005, shapeName=headstock_balloon_tie_capsule, center=<1.5 0 0>, rotation=<rot 18.9356 -90 2.0783>}]
 neck [{shapeType=Capsule, length=20.9321, radius=2.3497, shapeName=neck_capsule, center=<14 0.2 0.28>, rotation=<rot 0 89.8653 -89.8213>}]
 top_01 [{shapeType=Capsule, length=4.5, radius=3, shapeName=top_01_capsule, center=<5 0 0>, rotation=<rot 90 0 0>}]
 top_02 [{shapeType=Capsule, length=20.5496, radius=4.0851, shapeName=top_02_capsule, center=<14 0.8 2.6>, rotation=<rot 0 -89.9477 -89.7162>}]
 top_03 [{shapeType=Capsule, length=5.5, radius=9, shapeName=top_03_capsule, center=<9.8 1 0.5>, rotation=<rot 0 -90 0>}]
}
constraints [
 {bone1Name=headstock_balloon_01, bone2Name=neck, swing1LimitDegrees=10, swing2LimitDegrees=10, twistLimitDegrees=45, twistMotion=Locked}
 {bone1Name=headstock_balloon_02, bone2Name=neck, swing1LimitDegrees=10, swing2LimitDegrees=5, twistLimitDegrees=45, twistMotion=Locked}
 {
  bone1Name=headstock_balloon_tie
  bone2Name=headstock_balloon_02
  swing1LimitDegrees=10
  swing2LimitDegrees=10
  twistLimitDegrees=45
  twistMotion=Locked
 }
 {bone1Name=bottom_02, bone2Name=bottom_01, swing1LimitDegrees=2, swing2LimitDegrees=10, twistLimitDegrees=2}
 {bone1Name=bottom_01, bone2Name=handle, swing1LimitDegrees=6, swing2LimitDegrees=10, twistLimitDegrees=45, twistMotion=Locked}
 {bone1Name=top_01, bone2Name=handle, swing1LimitDegrees=4, swing2LimitDegrees=10, twistLimitDegrees=2}
 {bone1Name=top_02, bone2Name=top_01, swing1LimitDegrees=2, swing2LimitDegrees=10, twistLimitDegrees=45, twistMotion=Locked}
 {
  bone1Name=top_03
  bone2Name=top_02
  swing1LimitDegrees=45
  swing1Motion=Locked
  swing2LimitDegrees=45
  swing2Motion=Locked
  twistLimitDegrees=45
  twistMotion=Locked
 }
 {
  bone1Name=bottom_02
  bone2Name=top_03
  swing1LimitDegrees=45
  swing1Motion=Locked
  swing2LimitDegrees=45
  swing2Motion=Locked
  twistLimitDegrees=45
  twistMotion=Locked
 }
 {bone1Name=neck, bone2Name=handle, swing1LimitDegrees=10, swing2LimitDegrees=5, twistLimitDegrees=45, twistMotion=Locked}
]
