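# Animation Blueprint for the InflateMate keytar back bling, adding physics-driven secondary motion to its skeleton.
# The output runs the input pose through a RigidBody node using the backpack physics asset, simulated relative to the handle bone with damped inherited motion.
# The event graph sets GravityOverride by skydive and parachute state, but the output node's gravity stays fixed and never reads it.
# A front-end blend and a second RigidBody chain are disconnected from the output.
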AnimBlueprint /SparksCosmetics/Instrument/Keytar/InflateMate/Meshes/InflateMate_Keytar_Backpack_AnimBP
target_skeleton=&/InflateMate_Keytar_Skeleton
graphs [
 AnimGraph {nodes [
   n1 = Get bIsInFrontEnd()
   Component To Local(ComponentPose=Local To Component(LocalPose=Blend Poses by bool(bActiveValue=@n1) {blendType=Linear}))
   n2 = RigidBody(ComponentPose=Local To Component(LocalPose=Use cached pose 'cached pose'())) {
    bOverrideWorldGravity=true
    overridePhysicsAsset=&/InflateMate_Keytar_Physics
    simulationSpace=BaseBoneSpace
    alphaBoolBlend {blendOption=Linear}
    baseBoneRef {boneName=handle}
    componentAppliedLinearAccClamp=<3000 3000 3000>
    componentLinearAccScale=<0.4 0.4 0.4>
    componentLinearVelScale=<0.2 0.2 0.2>
    overrideWorldGravity {z=-900}
    simSpaceSettings {
     maxAngularAcceleration=500
     maxAngularVelocity=3
     maxLinearAcceleration=700
     maxLinearVelocity=500
     velocityScaleZ=0
     worldAlpha=0.5
     externalLinearDragV=<0.2 0.2 0.02>
    }
   }
   Component To Local(ComponentPose=@n2)
   n3 = RigidBody!(ComponentPose=Local To Component(LocalPose=Input Pose()), OverrideWorldGravity:Vector="0,0,-900", ExternalForce) -> (Pose) {
    alphaInputType=Bool
    bOverrideWorldGravity=true
    overridePhysicsAsset=&/InflateMate_Keytar_Backpack_Physics
    simulationSpace=BaseBoneSpace
    alphaBoolBlend {blendOption=Linear}
    baseBoneRef {boneName=handle}
    componentAppliedLinearAccClamp=<3000 3000 3000>
    componentLinearAccScale=<0.4 0.4 0.4>
    componentLinearVelScale=<0.2 0.2 0.2>
    overrideWorldGravity {z=-900}
    simSpaceSettings {
     maxAngularAcceleration=500
     maxAngularVelocity=3
     maxLinearAcceleration=700
     maxLinearVelocity=500
     velocityScaleZ=0
     worldAlpha=0.5
     externalLinearDragV=<0.2 0.2 0.02>
    }
   }
   Output Pose(Result=Component To Local(ComponentPose=@n3))
  ]}
 EventGraph {nodes [
   Event BlueprintUpdateAnimation()
   > n1 = Branch(Condition=Get bIsSkyDiving())
   > n2 = Branch(Condition=Get bIsParachuteOpened())
   > Set GravityOverride(GravityOverride="0,0,-400")
   Set GravityOverride(execute=@n2.else, GravityOverride="0,0,500")
   Set GravityOverride(execute=@n1.else, GravityOverride="0,0,-900")
   TryGetPawnOwner()
  ]}
]
variables [
 As Player Pawn Athena {type=PlayerPawn_Athena_C}
 Backpack_SkelMesh {type=SkeletalMeshComponent}
 GravityOverride {type=Vector}
 "Instrument Pickaxe " {type="array<B_Athena_Pickaxe_Generic_C>", default []}
 IsBackpackEquipped {type=boolean}
 IsPickaxeEquipped {type=boolean}
 NewVar {type=boolean}
]
